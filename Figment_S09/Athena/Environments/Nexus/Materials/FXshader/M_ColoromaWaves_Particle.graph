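# Translucent, unlit, two-sided mesh-particle material for Coloroma wave effects in the Nexus environment, with depth test disabled.
# Curve-driven rotation of world position, a dynamic parameter and bump-offset turbulence noise combine to index a color curve atlas row.
# A view-angle ramp curve shapes distortion, texture glow and opacity.
# Tints and particle color modulate the result, depth fade softens opacity, and a CameraOffset function is included.
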Material /Figment_S09/Athena/Environments/Nexus/Materials/FXshader/M_ColoromaWaves_Particle
graph {
 properties {
  bDisableDepthTest=true
  bUsedWithMeshParticles=true
  blendMode=BLEND_Translucent
  pixelDepthOffsetMode=PDOM_Legacy
  shadingModel=MSM_Unlit
  twoSided=true
  usedShadingModels=Unlit
 }
 nodes [
  CollectionParameter() {collection=/Game/Athena/Environments/Nexus/Materials/FXshader/MPC_Nexus, parameterName=SphereWave_TimeScale}
  MaterialFunctionCall(Offset Amount=ScalarParameter() {defaultValue=100, parameterName=Offset}) {materialFunction=/Engine/Functions/Engine_MaterialFunctions02/WorldPositionOffset/CameraOffset}
  n1 = ParticleColor()
  n2 = DynamicParameter()
  n3 = Add(A=@n2.R, B=Time())
  n4 = ScalarParameter() {defaultValue=0.1, parameterName=TimeScaleRot}
  n5 = CurveAtlasRowParameter(CurveTime=Frac(None=Multiply(A=@n3, B=@n4))) {atlas=&AthenaCurveAtlas01, curve=/Game/Athena/Environments/Nexus/Materials/FXshader/Curve_Aix, defaultValue=77}
  n6 = Constant() {r=0.5}
  n7 = ObjectPositionWS()
  n8 = WorldPosition()
  n9 = Multiply(
   A=ComponentMask(None=RotateAboutAxis(NormalizedRotationAxis=@n5, RotationAngle=@n6, PivotPoint=@n7, Position=@n8.XYZ)) {b=true}
   B=ScalarParameter() {defaultValue=0.0005, parameterName=RA_wide}
  )
  n10 = ScalarParameter() {defaultValue=0.025, parameterName=FresnelAffectTexture}
  n11 = VertexNormalWS()
  n12 = CameraVectorWS()
  n13 = CurveAtlasRowParameter(CurveTime=Abs(None=DotProduct(A=@n11, B=@n12))) {
   atlas=&AthenaCurveAtlas01
   curve=/Game/Athena/Environments/Nexus/Materials/FXshader/Curve_Ramps
   defaultValue=79
  }
  n14 = Add(A=Add(A=Add(A=@n9, B=@n2.R), B=Multiply(A=@n13.G, B=@n10)))
  n15 = TextureCoordinate() {uTiling=3}
  n16 = TextureSample(UVs=Panner(Coordinate=@n15) {speedX=0.2, speedY=0.2}) {samplerType=SAMPLERTYPE_LinearColor, texture=/Game/Effects/Fort_Effects/Gameplay/SmokeDustExplosions/Textures/T_NoisePack_LinearColor}
  n17 = TextureSample(UVs=Panner(Coordinate=@n15) {speedX=-0.2, speedY=-0.2}) {samplerType=SAMPLERTYPE_LinearColor, texture=&T_NoisePack_LinearColor}
  n18 = ComponentMask(None=TextureCoordinate() {uTiling=0.5, vTiling=0.5}) {g=true}
  n19 = MaterialFunctionCall(In=Abs(None=Sine(None=@n18)), Contrast=Constant() {r=0.1}) {materialFunction=&CheapContrast}
  n20 = Add(
   A=Multiply(A=Subtract(A=@n16.R, B=@n17.G), B=@n19.Result)
   B=Multiply(A=Subtract(A=@n13.R), B=MaterialFunctionCall() {materialFunction=&DebugTimeSine}.Result)
  )
  n21 = Multiply(A=@n19.Result, B=ScalarParameter() {defaultValue=0.25, parameterName=Hight})
  n22 = Add(B=BumpOffset(Coordinate=TextureCoordinate() {uTiling=8, vTiling=5}, Height=@n20, HeightRatioInput=@n21))
  n23 = Panner(Coordinate=@n22) {speedX=0.1, speedY=0.3}
  n24 = TextureSample(UVs=Panner(Coordinate=Multiply(A=@n22) {constB=1.618}) {speedX=0.1, speedY=0.3}) {samplerType=SAMPLERTYPE_LinearColor, texture=&T_NoisePack_LinearColor}
  n25 = WorldPosition()
  n26 = Divide(A=ComponentMask(None=TransformPosition(None=@n25.XYZ) {transformSourceType=TRANSFORMPOSSOURCE_World}) {g=true, r=true}, B=ObjectRadius())
  n27 = LinearInterpolate(
   A=Multiply(A=TextureSample(UVs=@n23) {texture=/Game/Effects/Fort_Effects/Textures/Plasma/T_Turbulence_Seamless}.R, B=@n24.R)
   B=TextureSample(UVs=BumpOffset(Coordinate=@n26, Height=@n20, HeightRatioInput=@n21)) {texture=&T_Turbulence_Seamless}.R
   Alpha=OneMinus(None=@n19.Result)
  )
  n28 = Multiply(A=@n27, B=ScalarParameter() {defaultValue=0.5, parameterName=NoiseScale})
  n29 = CurveAtlasRowParameter(CurveTime=Frac(None=Add(A=@n14, B=@n28))) {
   atlas=&AthenaCurveAtlas01
   curve=/Game/Athena/Environments/Nexus/Materials/FXshader/Curve_Color
   defaultValue=78
  }
  n30 = VectorParameter() {parameterName=Tint, defaultValue=<rgba 2.5781 9.1695 15 0.5>}
  n31 = Multiply(A=Multiply(A=Saturate(None=@n27), B=@n13.B)) {constB=5}
  n32 = Add(A=Multiply(A=@n29.A, B=@n30.A), B=@n31)
  Multiply(A=@n1.A, B=DepthFade(Opacity=Multiply(A=@n32, B=@n13.A), FadeDistance=ScalarParameter() {defaultValue=100, parameterName=DF_FD}))
  n33 = Multiply(A=@n29, B=@n30.RGB)
  n34 = ScalarParameter() {defaultValue=1.5, parameterName=Exp}
  n35 = VectorParameter() {parameterName=TextureTint, defaultValue=<rgba 0.0984 0.3942 1 1>}
  Multiply(A=@n1.RGB, B=Add(A=Power(Base=@n33, Exp=@n34), B=Multiply(A=@n31, B=@n35.RGB)))
  ScalarParameter() {parameterName=RA_Offset}
 ]
}
